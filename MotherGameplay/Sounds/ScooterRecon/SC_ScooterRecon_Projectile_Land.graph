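# Sound cue for the Scooter Recon projectile landing, played as a spatialized weapon sound.
# It mixes a random bounce sample, routed through a delay node, with a much quieter, pitched-down random pulse sample.
# An IsFriendly branch varies the pulse, making it louder and higher when false or unset, using one shared random wave set.
# Attenuation applies HRTF spatialization and distance low-pass filtering, and local and global concurrency sets limit voices.
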SoundCue /MotherGameplay/Sounds/ScooterRecon/SC_ScooterRecon_Projectile_Land
root=Mixer_0
nodes [
 {id=Branch_0, children [Modulator_0,Modulator_1,Modulator_1], properties {boolParameterName=IsFriendly}}
 {id=Delay_0, children [Random_0]}
 {id=Mixer_0, children [Delay_0,Modulator_2], properties {inputVolume [1,0.1]}}
 {id=Modulator_0, children [Random_2], properties {pitchMax=1, pitchMin=1, volumeMax=1, volumeMin=0.8}}
 {id=Modulator_1, children [Random_2], properties {pitchMax=1.2, pitchMin=1.2, volumeMax=2, volumeMin=2}}
 {id=Modulator_2, children [Branch_0], properties {pitchMax=0.5, pitchMin=0.5}}
 {id=Random_0, children [WavePlayer_0,WavePlayer_1,WavePlayer_2], properties {weights [1,1,1]}}
 {id=Random_2, children [WavePlayer_8, WavePlayer_9, WavePlayer_10], properties {weights [1,1,1]}}
 {id=WavePlayer_0, properties {soundWaveAssetPtr=&/ScooterRecon_Projectile_Bounce_01}}
 {id=WavePlayer_1, properties {soundWaveAssetPtr=&/ScooterRecon_Projectile_Bounce_03}}
 {id=WavePlayer_10, properties {soundWaveAssetPtr=&/ScooterRecon_Pulse_C03}}
 {id=WavePlayer_2, properties {soundWaveAssetPtr=&/ScooterRecon_Projectile_Bounce_04}}
 {id=WavePlayer_8, properties {soundWaveAssetPtr=&/ScooterRecon_Pulse_C01}}
 {id=WavePlayer_9, properties {soundWaveAssetPtr=&/ScooterRecon_Pulse_C02}}
]
properties {
 attenuationSettings=&Fort_Weapon_Projectile_Close_Athena_Att
 maxDistance=4950
 soundClassObject=/Game/Sounds/FortSoundClassesAndModes/Fort_Weapons
 volumeMultiplier=0.6
 attenuationOverrides {
  bAttenuateWithLPF=true
  falloffDistance=3500
  lPFFrequencyAtMax=200
  lPFRadiusMax=3000
  lPFRadiusMin=800
  spatializationAlgorithm=SPATIALIZATION_HRTF
  stereoSpread=256
 }
 concurrencyOverrides {bLimitToOwner=true, maxCount=2}
 concurrencySet [
  /ScooterGameplay/Sounds/ScooterRecon/ScooterRecon_Projectile_Local_Conc
  /ScooterGameplay/Sounds/ScooterRecon/ScooterRecon_Projectile_Global_Conc
 ]
}
