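# Sound cue for the ScooterRecon heavy bullet whip, played in the ranged-weapon projectile effects sound class.
# A Modulator slightly varies pitch and volume over a Quality Level switch that feeds Random nodes.
# The first two quality slots pick from four large whip waves, while the last two share a two-wave Random excluded from branch culling.
# Attenuation overrides disable distance falloff and set a wide stereo spread.
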SoundCue /MotherGameplay/Sounds/ScooterRecon/ScooterRecon_BulletWhipHeavy_Cue
root=Modulator_0
nodes [
 {id=Modulator_0, children [QualityLevel_0], properties {pitchMax=1.1, pitchMin=0.9, volumeMax=1.1, volumeMin=0.9}}
 {id=QualityLevel_0, children [Random_2,Random_2,Random_1,Random_1]}
 {id=Random_1, children [WavePlayer_4,WavePlayer_5], properties {bShouldExcludeFromBranchCulling=true, weights [1,1]}}
 {id=Random_2, children [WavePlayer_4,WavePlayer_5,WavePlayer_6,WavePlayer_7], properties {weights [1,1,1,1]}}
 {id=WavePlayer_4, properties {soundWaveAssetPtr=&/ScooterRecon_BulletWhip_Large_01}}
 {id=WavePlayer_5, properties {soundWaveAssetPtr=&/ScooterRecon_BulletWhip_Large_02}}
 {id=WavePlayer_6, properties {soundWaveAssetPtr=&/ScooterRecon_BulletWhip_Large_03}}
 {id=WavePlayer_7, properties {soundWaveAssetPtr=&/ScooterRecon_BulletWhip_Large_04}}
]
properties {
 bOverrideAttenuation=true
 maxDistance=2097152
 soundClassObject=&Fort_WeaponRanged_Projectile_Effects
 volumeMultiplier=1
 attenuationOverrides {bAttenuate=false, stereoSpread=5000}
}
