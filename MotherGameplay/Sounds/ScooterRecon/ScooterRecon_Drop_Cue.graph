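# Sound cue for the drop sound of the ScooterRecon item, classed under loot collection audio.
# A single wave player feeds a modulator left at its default settings, played at half volume.
# It uses the shared weapon drop attenuation and a weapon drop stop-oldest concurrency set, alongside a prevent-new concurrency override.
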SoundCue /MotherGameplay/Sounds/ScooterRecon/ScooterRecon_Drop_Cue
root=Modulator_0
nodes [
 {id=Modulator_0, children [WavePlayer_0]}
 {id=WavePlayer_0, properties {soundWaveAssetPtr=&/ScooterRecon_Drop_01}}
]
properties {
 attenuationSettings=&Fort_Weapon_Drop_Att
 maxDistance=3000
 soundClassObject=&Fort_Loot_Collect
 volumeMultiplier=0.5
 concurrencyOverrides {resolutionRule=PreventNew}
 concurrencySet [&Weapon_Drop_StopOldest]
}
