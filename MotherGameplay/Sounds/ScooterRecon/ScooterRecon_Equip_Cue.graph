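# Sound cue for the ScooterRecon equip foley, presumably heard when that item is equipped.
# A single wave player feeds a modulator that varies volume slightly while pinning pitch.
# The cue plays at reduced volume under the weapon foley sound class, borrows the weapon reload attenuation, overrides concurrency, and has its debug flag on.
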SoundCue /MotherGameplay/Sounds/ScooterRecon/ScooterRecon_Equip_Cue
root=Modulator_0
nodes [
 {id=Modulator_0, children [WavePlayer_0], properties {pitchMax=1, pitchMin=1, volumeMax=1.1}}
 {id=WavePlayer_0, properties {soundWaveAssetPtr=&/ScooterRecon_Equip_01}}
]
properties {
 attenuationSettings=&Fort_Weapon_Reload_Att
 bDebug=true
 bOverrideConcurrency=true
 maxDistance=2250
 soundClassObject=&Fort_WeaponFoley
 volumeMultiplier=0.401
}
