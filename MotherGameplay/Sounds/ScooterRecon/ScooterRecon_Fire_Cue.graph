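# Sound cue for ScooterRecon weapon fire, layering close, medium and distant gunshot sounds.
# A mixer blends the three layers, each with its own attenuation and sound class, under global and local gunfire concurrency sets.
# The first quality level randomizes three waves per layer; the other levels share a mix that keeps close variants but plays single medium and distant waves.
# Modulators vary pitch on every layer except the first level's medium one.
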SoundCue /MotherGameplay/Sounds/ScooterRecon/ScooterRecon_Fire_Cue
root=QualityLevel_0
nodes [
 {
  id=Attenuation_10
  children [Random_7]
  properties {attenuationSettings=&Fort_Weapon_Projectile_Medium_Mono_Athena_Att}
 }
 {
  id=Attenuation_11
  children [Modulator_7]
  properties {attenuationSettings=&Fort_Weapon_Projectile_Distant_Mono_Athena_Att}
 }
 {id=Attenuation_3, children [Modulator_2], properties {attenuationSettings=&Fort_Weapon_Projectile_Close_Athena_Att}}
 {
  id=Attenuation_4
  children [Modulator_4]
  properties {attenuationSettings=&Fort_Weapon_Projectile_Medium_Mono_Athena_Att}
 }
 {
  id=Attenuation_5
  children [Modulator_3]
  properties {attenuationSettings=&Fort_Weapon_Projectile_Distant_Mono_Athena_Att}
 }
 {id=Attenuation_9, children [Modulator_6], properties {attenuationSettings=&Fort_Weapon_Projectile_Close_Athena_Att}}
 {id=Mixer_1, children [SoundClass_7,SoundClass_8,SoundClass_5], properties {inputVolume [1,0.8,0.5]}}
 {id=Mixer_3, children [SoundClass_1,SoundClass_2,SoundClass_6], properties {inputVolume [1,0.8,0.5]}}
 {id=Modulator_2, children [Random_8], properties {pitchMax=1.1, pitchMin=1}}
 {id=Modulator_3, children [WavePlayer_33], properties {pitchMax=1.1}}
 {id=Modulator_4, children [WavePlayer_36], properties {pitchMax=1.1}}
 {id=Modulator_6, children [Random_4], properties {pitchMax=1.1, pitchMin=1}}
 {id=Modulator_7, children [Random_6], properties {pitchMax=1.1}}
 {id=QualityLevel_0, children [Mixer_3,Mixer_1,Mixer_1,Mixer_1]}
 {id=Random_4, children [WavePlayer_15, WavePlayer_17, WavePlayer_18], properties {weights [1,1,1]}}
 {id=Random_6, children [WavePlayer_24, WavePlayer_25, WavePlayer_26], properties {weights [1,1,1]}}
 {id=Random_7, children [WavePlayer_27, WavePlayer_28, WavePlayer_29], properties {weights [1,1,1]}}
 {id=Random_8, children [WavePlayer_30, WavePlayer_31, WavePlayer_32], properties {weights [1,1,1]}}
 {
  id=SoundClass_1
  children [Attenuation_9]
  properties {soundClassOverride=&Fort_WeaponRanged_Fire_3P_PSM}
 }
 {
  id=SoundClass_2
  children [Attenuation_10]
  properties {soundClassOverride=&Fort_WeaponRanged_Fire_Distant}
 }
 {
  id=SoundClass_5
  children [Attenuation_5]
  properties {soundClassOverride=&Fort_WeaponRanged_Fire_Distant}
 }
 {
  id=SoundClass_6
  children [Attenuation_11]
  properties {soundClassOverride=&Fort_WeaponRanged_Fire_Distant}
 }
 {
  id=SoundClass_7
  children [Attenuation_3]
  properties {soundClassOverride=&Fort_WeaponRanged_Fire_3P_PSM}
 }
 {
  id=SoundClass_8
  children [Attenuation_4]
  properties {soundClassOverride=&Fort_WeaponRanged_Fire_Distant}
 }
 {id=WavePlayer_15, properties {soundWaveAssetPtr=&/ScooterRecon_Fire_Close_01}}
 {id=WavePlayer_17, properties {soundWaveAssetPtr=&/ScooterRecon_Fire_Close_02}}
 {id=WavePlayer_18, properties {soundWaveAssetPtr=&/ScooterRecon_Fire_Close_03}}
 {id=WavePlayer_24, properties {soundWaveAssetPtr=&/ScooterRecon_Fire_Distant_01}}
 {id=WavePlayer_25, properties {soundWaveAssetPtr=&/ScooterRecon_Fire_Distant_02}}
 {id=WavePlayer_26, properties {soundWaveAssetPtr=&/ScooterRecon_Fire_Distant_03}}
 {id=WavePlayer_27, properties {soundWaveAssetPtr=&/ScooterRecon_Fire_Medium_01}}
 {id=WavePlayer_28, properties {soundWaveAssetPtr=&/ScooterRecon_Fire_Medium_02}}
 {id=WavePlayer_29, properties {soundWaveAssetPtr=&/ScooterRecon_Fire_Medium_03}}
 {id=WavePlayer_30, properties {soundWaveAssetPtr=&ScooterRecon_Fire_Close_01}}
 {id=WavePlayer_31, properties {soundWaveAssetPtr=&ScooterRecon_Fire_Close_02}}
 {id=WavePlayer_32, properties {soundWaveAssetPtr=&ScooterRecon_Fire_Close_03}}
 {id=WavePlayer_33, properties {soundWaveAssetPtr=&ScooterRecon_Fire_Distant_01}}
 {id=WavePlayer_36, properties {soundWaveAssetPtr=&ScooterRecon_Fire_Medium_01}}
]
properties {
 maxDistance=27500
 pitchMultiplier=1.05
 soundClassObject=&Fort_WeaponRanged_Fire_3P_PSM
 volumeMultiplier=0.8
 attenuationOverrides {falloffDistance=5000, attenuationShapeExtents {x=1000}}
 concurrencyOverrides {maxCount=3}
 concurrencySet [&Weapon_Gunfire_Global_Conc, &Weapon_Gunfire_Local_Conc]
}
