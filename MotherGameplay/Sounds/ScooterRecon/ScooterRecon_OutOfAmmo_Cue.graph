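# Sound cue for the out-of-ammo sound of the ScooterRecon weapon.
# A random node picks one of three equally weighted wave variants, then a modulator slightly varies volume while pitch stays fixed.
# Cue-level multipliers lower volume and raise pitch.
# It uses a third-person ranged-fire sound class, a short attenuation falloff override, and concurrency capped at two instances that stops the oldest.
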SoundCue /MotherGameplay/Sounds/ScooterRecon/ScooterRecon_OutOfAmmo_Cue
root=Modulator_1
nodes [
 {id=Modulator_1, children [Random_1], properties {pitchMax=1, pitchMin=1, volumeMax=0.7, volumeMin=0.6}}
 {id=Random_1, children [WavePlayer_0,WavePlayer_5,WavePlayer_6], properties {weights [1,1,1]}}
 {id=WavePlayer_0, properties {soundWaveAssetPtr=&/ScooterRecon_OutOfAmmo_01}}
 {id=WavePlayer_5, properties {soundWaveAssetPtr=&/ScooterRecon_OutOfAmmo_02}}
 {id=WavePlayer_6, properties {soundWaveAssetPtr=&/ScooterRecon_OutOfAmmo_03}}
]
properties {
 bOverrideConcurrency=true
 maxDistance=2097152
 pitchMultiplier=1.2
 soundClassObject=&Fort_WeaponRanged_Fire_3P_PSM
 volumeMultiplier=0.4
 attenuationOverrides {falloffDistance=1600}
 concurrencyOverrides {maxCount=2, resolutionRule=StopOldest}
}
