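# Sound cue for picking up the ScooterRecon item, categorized as loot-collect audio.
# A single wave player feeds the root directly, playing one pickup sample with reduced volume and slightly raised pitch.
# It uses the shared weapon pickup attenuation and overrides concurrency to allow one instance at a time, preventing new plays while one is active.
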SoundCue /MotherGameplay/Sounds/ScooterRecon/ScooterRecon_Pickup_Cue
root=WavePlayer_0
nodes [{id=WavePlayer_0, properties {soundWaveAssetPtr=&/ScooterRecon_Pickup_01}}]
properties {
 attenuationSettings=&Fort_Weapon_Pickup_Att
 bOverrideConcurrency=true
 maxDistance=3500
 pitchMultiplier=1.1
 soundClassObject=&Fort_Loot_Collect
 volumeMultiplier=0.279
 concurrencyOverrides {maxCount=1, resolutionRule=PreventNew}
}
