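# Sound cue for the bounce of the ScooterRecon projectile, in the ranged weapon projectile effects sound class.
# An equally weighted Random node picks one of four bounce waves and feeds a Modulator that slightly varies pitch.
# Inline attenuation overrides the referenced grenade bounce asset, adding distance falloff and low-pass filtering.
# Concurrency is overridden to allow two voices per owner, stopping the oldest.
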SoundCue /MotherGameplay/Sounds/ScooterRecon/ScooterRecon_Projectile_Bounce_Cue
root=Modulator_5
nodes [
 {id=Modulator_5, children [Random_1], properties {pitchMax=1.1, pitchMin=0.9}}
 {id=Random_1, children [WavePlayer_6, WavePlayer_8, WavePlayer_9, WavePlayer_11], properties {weights [1,1,1,1]}}
 {id=WavePlayer_11, properties {soundWaveAssetPtr=&/ScooterRecon_Projectile_Bounce_06}}
 {id=WavePlayer_6, properties {soundWaveAssetPtr=&/ScooterRecon_Projectile_Bounce_01}}
 {id=WavePlayer_8, properties {soundWaveAssetPtr=&/ScooterRecon_Projectile_Bounce_03}}
 {id=WavePlayer_9, properties {soundWaveAssetPtr=&/ScooterRecon_Projectile_Bounce_04}}
]
properties {
 attenuationSettings=/Game/Sounds/Attenuation/Fort_Grenade_Bounce_Att
 bOverrideAttenuation=true
 bOverrideConcurrency=true
 maxDistance=3600
 soundClassObject=&Fort_WeaponRanged_Projectile_Effects
 volumeMultiplier=1
 attenuationOverrides {bAttenuateWithLPF=true, falloffDistance=3000, lPFFrequencyAtMax=600, lPFRadiusMax=3000, lPFRadiusMin=800, attenuationShapeExtents {x=600}}
 concurrencyOverrides {bLimitToOwner=true, maxCount=2, resolutionRule=StopOldest}
}
