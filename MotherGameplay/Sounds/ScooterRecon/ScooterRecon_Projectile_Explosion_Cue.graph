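# Sound cue for the explosion of the ScooterRecon projectile, mixing a close layer and a distant layer, each with its own projectile attenuation.
# The close layer blends one of three random explosion waves with a quieter scanning-orb sound through a modulator.
# The distant layer picks one of three random distant explosion waves, slightly lowered in volume.
# Playback uses the large explosives sound class, limited to one instance per owner, stopping the oldest.
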SoundCue /MotherGameplay/Sounds/ScooterRecon/ScooterRecon_Projectile_Explosion_Cue
root=Mixer_2
nodes [
 {
  id=Attenuation_2
  children [Modulator_1]
  properties {attenuationSettings=&Fort_Weapon_Projectile_Medium_Mono_Athena_Att}
 }
 {id=Attenuation_3, children [Modulator_0], properties {attenuationSettings=&Fort_Weapon_Projectile_Close_Athena_Att}}
 {id=Mixer_0, children [Random_4,WavePlayer_1], properties {inputVolume [1.2,0.2]}}
 {id=Mixer_1, children [Random_5,null], properties {inputVolume [1.2,0.7]}}
 {id=Mixer_2, children [Attenuation_3, Attenuation_2], properties {inputVolume [1,0.7]}}
 {id=Modulator_0, children [Mixer_0]}
 {id=Modulator_1, children [Mixer_1], properties {volumeMax=0.9, volumeMin=0.8}}
 {id=Random_4, children [WavePlayer_6, WavePlayer_11, WavePlayer_12], properties {weights [1,1,1]}}
 {id=Random_5, children [WavePlayer_13, WavePlayer_14, WavePlayer_15], properties {weights [1,1,1]}}
 {id=WavePlayer_1, properties {soundWaveAssetPtr=&/ScooterRecon_Projectile_ScanningOrb_Close_02}}
 {id=WavePlayer_11, properties {soundWaveAssetPtr=&/ScooterRecon_Projectile_Explosion_02}}
 {id=WavePlayer_12, properties {soundWaveAssetPtr=&/ScooterRecon_Projectile_Explosion_03}}
 {id=WavePlayer_13, properties {soundWaveAssetPtr=&/ScooterRecon_Projectile_ExplosionDistant_01}}
 {id=WavePlayer_14, properties {soundWaveAssetPtr=&/ScooterRecon_Projectile_ExplosionDistant_02}}
 {id=WavePlayer_15, properties {soundWaveAssetPtr=&/ScooterRecon_Projectile_ExplosionDistant_03}}
 {id=WavePlayer_6, properties {soundWaveAssetPtr=&/ScooterRecon_Projectile_Explosion_01}}
]
properties {
 bOverrideConcurrency=true
 maxDistance=10579.2871
 soundClassObject=/Game/Sounds/FortSoundClassesAndModes/Fort_Explosives_Large_PSM
 volumeMultiplier=0.7784
 concurrencyOverrides {bLimitToOwner=true, maxCount=1, resolutionRule=StopOldest}
}
