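# Sound cue for the ScooterRecon projectile's near-death sound.
# A single wave player feeds a modulator that applies a fixed gain boost without randomization, under a low cue volume and slightly raised pitch.
# Overridden attenuation adds distance-based low-pass filtering, and overridden concurrency caps simultaneous instances at three.
# It routes through a third-person ranged weapon fire sound class.
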SoundCue /MotherGameplay/Sounds/ScooterRecon/ScooterRecon_Projectile_NearDeath_Cue
root=Modulator_0
nodes [
 {id=Modulator_0, children [WavePlayer_3], properties {pitchMax=1, pitchMin=1, volumeMax=1.5, volumeMin=1.5}}
 {id=WavePlayer_3, properties {soundWaveAssetPtr=&/ScooterRecon_Projectile_NearDeath_mix}}
]
properties {
 attenuationSettings=/Game/Sounds/Attenuation/Fort_GadgetWorking_Att
 bOverrideAttenuation=true
 bOverrideConcurrency=true
 maxDistance=3000
 pitchMultiplier=1.1
 soundClassObject=&Fort_WeaponRanged_Fire_3P_PSM
 volumeMultiplier=0.17
 attenuationOverrides {
  bAttenuateWithLPF=true
  falloffDistance=2400
  lPFFrequencyAtMax=600
  lPFRadiusMax=3000
  lPFRadiusMin=1000
  stereoSpread=1
  attenuationShapeExtents {x=600}
 }
 concurrencyOverrides {maxCount=3}
}
