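# Sound cue named as the in-air loop for the ScooterRecon scanning orb projectile, but it is effectively silent, with no wave attached and a zero volume multiplier.
# A SoundClass node overrides the class to a third-person ranged weapon fire class over a slight pitch Modulator whose input is empty.
# The cue carries its own long-range attenuation with distance low-pass and reverb, plus a concurrency rule that prevents new instances at the limit.
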SoundCue /MotherGameplay/Sounds/ScooterRecon/ScooterRecon_Projectile_ScanningOrb_InAirLoop_Cue
root=SoundClass_3
nodes [
 {id=Modulator_3, children [null], properties {pitchMax=1.1, pitchMin=1}}
 {id=SoundClass_3, children [Modulator_3], properties {soundClassOverride=&Fort_WeaponRanged_Fire_3P_PSM}}
]
properties {
 bOverrideAttenuation=true
 bOverrideConcurrency=true
 maxDistance=8800
 soundClassObject=/Game/Sounds/FortSoundClassesAndModes/Fort_Weapons
 volumeMultiplier=0
 attenuationOverrides {
  bAttenuateWithLPF=true
  falloffDistance=8000
  lPFFrequencyAtMax=600
  lPFRadiusMax=3000
  lPFRadiusMin=800
  reverbWetLevelMax=0.7
  reverbWetLevelMin=0.2
  stereoSpread=5000
  attenuationShapeExtents {x=800}
 }
 concurrencyOverrides {resolutionRule=PreventNew}
}
